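; Gameplay ability that tracks how far the owning player travels while sand tunneling and reports it as quest progress.
; A looping timer with a randomized interval, chosen for performance, checks the sand tunneling tag and adds whole meters of XY movement.
; Without the tag it resets the stored locations.
; Totals go to the quest manager as a custom stat event once past a threshold, and also when the player dies.
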(event Ability|EventActivateAbility
  (bind _owningplayerpawn (Variables|GAVars|Data|GetOwningPlayerPawn))
  ;[1 1 1] Activate Ability / Init
  (|CallParentFunction)
  (Utilities|IsValid _owningplayerpawn
    (:Is%20Valid
      (bind _output_get (Variables|Default|SetPreviousLocation (Transformation|GetActorLocation _owningplayerpawn)))
      (Variables|Default|SetCurrentLocation _output_get)
      ;[1 1 1] Tick / Timer
      (EventDispatchers|BindEventtoOnDied _owningplayerpawn (EventDispatchers|CreateEvent :!event Custom|Died))
      (bind _output_get_2 (Variables|Default|SetInterval (Math|Random|RandomFloatinRange 1.7 2.5))) ; Tick at close random intervals for perf, do not go under 1 second
      (Utilities|Time|SetTimerbyEvent (delegate Custom|SendDistanceUpdate) _output_get_2 true))))

(event Custom|Died
  (params (DamagedActor :type Actor%20Object%20Reference) (Damage :type Float%20%28single-precision%29) (InstigatedBy :type Controller%20Object%20Reference) (DamageCauser :type Actor%20Object%20Reference) (HitLocation :type Vector) (FHitComponent :type Primitive%20Component%20Object%20Reference) (BoneName :type Name) (Momentum :type Vector))
  ;[1 1 1] Player Died
  (CallFunction|QuestUpdate))

(event Custom|Quest%20Update
  ;[1 1 1] Send quest progress
  (Quest|SendComplexCustomStatEvent (Variables|GAVars|Data|GetOwningPlayerQuestManager) 0 "(GameplayTags=((TagName=\"Athena.Quests.Distance.SandTunneling\")))" "(GameplayTags=)" (Variables|Default|GetMetersTravelled))
  (Variables|Default|SetMetersTravelled 0))

(event Custom|SendDistanceUpdate
  (bind (_previouslocation_x _previouslocation_y) (Math|Vector|BreakVector (Variables|GetPreviousLocation_X)))
  (bind (_currentlocation_x _currentlocation_y) (Math|Vector|BreakVector (Variables|GetCurrentLocation_X)))
  (bind _meterstravelled (Variables|Default|GetMetersTravelled))
  ;[1 1 1] Make sure player is moving in the sand
  (Utilities|IsValid (Variables|GAVars|Data|GetOwningPlayerPawn)
    (:Is%20Valid
      (bind _returnvalue (MissionHelpers|HasTags (GameplayTags|GetOwnedGameplayTags (Variables|GAVars|Data|GetOwningPlayerPawn) :!class /Script/GameplayTags.GameplayTagAssetInterface) (Variables|Default|GetT_SandTunneling) false false))
      (if _returnvalue
        ;[1 1 1] Calculate distance travelled along XY plane
        (Variables|Default|SetCurrentLocation (Transformation|GetActorLocation (Variables|GAVars|Data|GetOwningPlayerPawn)))
        (Variables|Default|SetMetersTravelled (+ (Math|Float|Floor (/ (Math|Vector|VectorLength (Math|Vector|vector-vector (Math|Vector|MakeVector _currentlocation_x _currentlocation_y) (Math|Vector|MakeVector _previouslocation_x _previouslocation_y))) 1e+02)) (Variables|Default|GetMetersTravelled)))
        (Variables|Default|SetPreviousLocation (Variables|Default|GetCurrentLocation)) ; Set the current location to the previous location
        ;[1 1 1] Send quest progress
        (if (> _meterstravelled 50)
          (Development|PrintString (Utilities|String|Append "LogFortQuest: Sand Tunnel Distance Update - " (Utilities|String|ToString(Integer) _meterstravelled) " meters") false)
          (Quest|SendComplexCustomStatEvent (Variables|GAVars|Data|GetOwningPlayerQuestManager) 0 "(GameplayTags=((TagName=\"Athena.Quests.Distance.SandTunneling\")))" "(GameplayTags=)" (Variables|Default|GetMetersTravelled))
          (Variables|Default|SetMetersTravelled 0))
        (else
          (bind _output_get (Variables|Default|SetPreviousLocation (Transformation|GetActorLocation (Variables|GAVars|Data|GetOwningPlayerPawn))))
          (Variables|Default|SetCurrentLocation _output_get) ; Reset prev and curr locations while player is not sand tunneling
        )))) ; pawn valid?
)

(event Ability|EventFailedToActivatePassiveAbility
  (params (ActorInfo :type Gameplay%20Ability%20Actor%20Info%20Structure)))
